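# Unlit masked material for Vive controller trim lines in VR Editor, with an amplified color tint (blue by default).
# Floor and clamp math splits U into five bands, each weighted by its own scalar (B1 to B5) for separate control.
# Areas with U or V of 1 or above are added regardless of those scalars.
# The summed mask is multiplied by the clamped tangent-space camera vector Z, so it depends on view angle.
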Material /Engine/VREditor/Devices/Vive/M_ViveControllerTrim
graph {
 properties {
  blendMode=BLEND_Masked
  pixelDepthOffsetMode=PDOM_Legacy
  previewMesh=&/VivePreController_Trimlines
  shadingModel=MSM_Unlit
  usedShadingModels=Unlit
 }
 nodes [
  Constant() {r=1}
  Multiply(A=ScalarParameter() {defaultValue=2, parameterName=Amplify}, B=VectorParameter() {parameterName=Color, defaultValue=<rgba 0 0.5859 1 0>}.RGB)
  n1 = Transform(None=CameraVectorWS()) {transformSourceType=TRANSFORMSOURCE_World, transformType=TRANSFORM_Tangent}
  n2 = ScalarParameter() {defaultValue=1, parameterName=B4}
  n3 = ComponentMask(None=TextureCoordinate() {uTiling=5}) {r=true}
  n4 = Clamp(None=OneMinus(None=Floor(None=@n3)))
  n5 = ScalarParameter() {defaultValue=1, parameterName=B3}
  n6 = Add(A=@n3) {constB=-1}
  n7 = Clamp(None=OneMinus(None=Floor(None=@n6)))
  n8 = Subtract(A=@n7, B=@n4)
  n9 = ScalarParameter() {defaultValue=1, parameterName=B5}
  n10 = Add(A=@n6) {constB=-1}
  n11 = Clamp(None=OneMinus(None=Floor(None=@n10)))
  n12 = Subtract(A=@n11, B=@n7)
  n13 = ScalarParameter() {defaultValue=1, parameterName=B2}
  n14 = Add(A=@n10) {constB=-1}
  n15 = Clamp(None=OneMinus(None=Floor(None=@n14)))
  n16 = Subtract(A=@n15, B=@n11)
  n17 = Add(A=Add(A=Multiply(A=@n2, B=@n4), B=Multiply(A=@n5, B=@n8)), B=Add(A=Multiply(A=@n9, B=@n12), B=Multiply(A=@n13, B=@n16)))
  n18 = Add(A=@n14) {constB=-1}
  n19 = Clamp(None=OneMinus(None=Floor(None=@n18)))
  n20 = TextureCoordinate()
  n21 = Floor(None=ComponentMask(None=@n20) {r=true})
  n22 = Floor(None=ComponentMask(None=@n20) {g=true})
  n23 = Add(
   A=Multiply(A=ScalarParameter() {defaultValue=1, parameterName=B1}, B=Subtract(A=@n19, B=@n15))
   B=Max(A=Clamp(None=@n21), B=Clamp(None=@n22))
  )
  Multiply(A=ComponentMask(None=Clamp(None=@n1)) {b=true}, B=Clamp(None=Add(A=@n17, B=@n23)))
 ]
}
