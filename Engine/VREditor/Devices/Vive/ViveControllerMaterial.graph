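# Default Lit material for the HTC Vive controller model in the VR Editor.
# It samples the ViveControllerDiffuse texture and multiplies its RGB by a small constant, heavily darkening the color.
# The graph shows no other textures, parameters or nodes.
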Material /Engine/VREditor/Devices/Vive/ViveControllerMaterial
graph {
 properties {pixelDepthOffsetMode=PDOM_Legacy, usedShadingModels=Default Lit}
 nodes [Multiply(A=TextureSample() {texture=&/ViveControllerDiffuse}.RGB) {constB=0.02}]
}
