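# Material for the HTC Vive Pre controller model in VR Editor, with glowing trim, button highlights and a power light.
# A glow chain sums switchable trim (squared TrimGlowColor times a mask), five button masks weighted by B1–B5 and tinted InteractColor, and a colored power light.
# It also holds diffuse and specular textures, a panning laser texture, and a sine-based world-space grid that uses only X and Z.
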Material /Engine/VREditor/Devices/Vive/VivePreControllerMaterial
graph {
 properties {pixelDepthOffsetMode=PDOM_Legacy, previewMesh=&/VivePreControllerMesh, usedShadingModels=Default Lit}
 nodes [
  n1 = VectorParameter() {parameterName=TrimGlowColor, defaultValue=<rgba 0 0.4685 1 3>}
  n2 = Multiply(A=@n1.RGB, B=@n1.RGB)
  n3 = TextureSample() {texture=&/T_VivePreController_EM}
  n4 = VectorParameter() {group=Colors, parameterName=InteractColor, defaultValue=<rgba 1 1 1 0>}
  n5 = TextureSample() {texture=&/VivePreController_MasksA}
  n6 = Multiply(A=@n5.R, B=ScalarParameter() {group=Buttons, parameterName=B1})
  n7 = Multiply(A=@n5.G, B=ScalarParameter() {group=Buttons, parameterName=B2})
  n8 = ScalarParameter() {group=Buttons, parameterName=B3}
  n9 = ScalarParameter() {group=Buttons, parameterName=B4}
  n10 = TextureSample(UVs=TextureCoordinate() {uTiling=2, vTiling=2}) {texture=&/VivePreController_MasksB}
  n11 = ScalarParameter() {group=Buttons, parameterName=B5}
  n12 = Add(A=Add(A=Add(A=@n6, B=@n7), B=Multiply(A=@n5.B, B=@n8)), B=Add(A=Multiply(A=@n9, B=@n10.R), B=Multiply(A=@n11, B=@n10.G)))
  n13 = Multiply(
   A=TextureSample(UVs=TextureCoordinate() {uTiling=2, vTiling=2}) {texture=&VivePreController_MasksB}.B
   B=ScalarParameter() {defaultValue=100, group=Power Light, parameterName=Power}
  )
  n14 = VectorParameter() {group=Power Light, parameterName=PowerColor, defaultValue=<rgba 0.5246 0.8842 1 0>}
  Add(
   A=StaticSwitchParameter(True=Multiply(A=@n2, B=@n3.R), False=Constant()) {defaultValue=true, parameterName=UseGlowingTrim}
   B=Add(A=Multiply(A=@n4.RGB, B=@n12), B=Multiply(A=@n13, B=@n14.RGB))
  )
  Constant() {r=0.8}
  LinearInterpolate(A=Constant3Vector() {constant=<rgba 0.445 0.445 0.445 1>}, B=Constant3Vector() {constant=<rgba 0.055 0.055 0.055 1>})
  LinearInterpolate() {constA=0.45, constB=0.7}
  Max(A=TextureSample() {texture=&/VivePreControllerDiffuse}.RGB) {constB=0.012}
  n15 = Divide(A=WorldPosition().XYZ, B=ScalarParameter() {defaultValue=2, group=Grid, parameterName=GridSpacing})
  n16 = MaterialFunctionCall(Float3=Clamp(None=Sine(None=@n15) {period=2})) {materialFunction=&BreakOutFloat3Components}
  n17 = Max(A=Max(A=@n16.R, B=@n16.R), B=@n16.B)
  n18 = Constant() {r=1}
  n19 = ScalarParameter() {defaultValue=0.1, group=Grid, parameterName=Grid Width}
  Multiply(
   A=@n4.RGB
   B=Add(A=SphereMask(A=@n17, B=@n18, Radius=@n19) {attenuationRadius=0.1, hardnessPercent=0}, B=ScalarParameter() {group=Grid, parameterName=Bias}) {constB=0.05}
  )
  TextureSample(UVs=Panner(Coordinate=TextureCoordinate() {vTiling=0}) {speedY=0.1}) {texture=/Engine/VREditor/LaserPointer/VR_LaserPower_01}
  TextureSample() {texture=&/VivePreControllerSpec}
 ]
}
